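# One-shot stereo MetaSound for additive Kicks cosmetic foley, using the player foley sound class.
# After an adjustable delay it plays a random high-intensity sound, with pitch and volume randomized around base values.
# When playback starts, an affiliation comparison routes output to either the gain-only signal or one processed by spatialization and footstep distance-curve patches.
# The same comparison also switches the azimuth, distance and curve-enable values feeding those patches.
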MetaSoundSource /FortMetasound/Cosmetics/Kicks/MSS_Kicks_Additive_Foley
inputs [
 Character.Affiliation {type=Enum:FortAudioAffiliation}
 Delay Time {type=Time}
 HighIntensity.Sounds {type=WaveAsset:Array}
 Random.Pitch.Base {type=Float}
 Random.Pitch.Spread {type=Float}
 Random.Volume.Base {type=Float}
 Random.Volume.Spread {type=Float}
 UE.Attenuation.Distance {type=Float}
 UE.Source.OnPlay {type=Trigger}
 UE.Spatialization.Azimuth {type=Float}
]
interfaces [UE.OutputFormat.Stereo:1.0, UE.Source.OneShot:1.0, UE.Source:1.1]
outputs [
 "UE.OutputFormat.Stereo.Audio:0" {type=Audio}
 "UE.OutputFormat.Stereo.Audio:1" {type=Audio}
 UE.Source.OneShot.OnFinished {type=Trigger}
]
pages [Default {
  input_defaults {
   Character.Affiliation=0
   Delay Time=0
   Random.Pitch.Base=0
   Random.Pitch.Spread=0
   Random.Volume.Base=0
   Random.Volume.Spread=0
   UE.Attenuation.Distance=0
   UE.Source.OnPlay=null
   UE.Spatialization.Azimuth=0
   HighIntensity.Sounds []
  }
  nodes [
   InitVariable.Trigger()
   TriggerDelay_0 = UE.Trigger Delay(Delay Time=^Delay Time, In=^UE.Source.OnPlay)
   ArrayRandomGet_0 = Array.Random Get:WaveAsset:Array(In Array=^HighIntensity.Sounds, Next=@TriggerDelay_0.Out) {also {defaults {No Repeats=-1, Weights []}}}
   n1 = AbsoluteValue.Abs:Float(Input=^Random.Pitch.Spread)
   WavePlayer_0 = UE.Wave Player:Stereo(
    Play=@ArrayRandomGet_0.On Next
    Wave Asset=@ArrayRandomGet_0.Value
    Pitch Shift=UE.RandomFloat(
     Next=^UE.Source.OnPlay
     Max=UE.Add:Float(PrimaryOperand=^Random.Pitch.Base, AdditionalOperands=@n1.Value).Out
     Min=UE.Subtract:Float(PrimaryOperand=^Random.Pitch.Base, AdditionalOperands=@n1.Value).Out
    ).Value
   ) {also {version [1,1]}}
   n2 = &TriggerCompareAffiliation(
    A=^Character.Affiliation
    B=VariableAccessor.Enum:FortAudioAffiliation(Variable=InitVariable.Enum:FortAudioAffiliation().Variable).Value
    Compare=@WavePlayer_0.On Play
   )
   n3 = InitVariable.Audio()
   n4 = VariableAccessor.Trigger(Variable=VariableMutator.Trigger(Variable=InitVariable.Trigger().Variable, Value=@n2.True).Variable)
   n5 = VariableAccessor.Trigger(Variable=VariableMutator.Trigger(Variable=InitVariable.Trigger().Variable, Value=@n2.False).Variable)
   TriggerRouteTriggerRouteFloat2_0 = "TriggerRoute.Trigger Route (Float, 2):Float"(Value 0=^UE.Spatialization.Azimuth, Set 0=@n4.Value, Set 1=@n5.Value) {also {defaults {Value 1=0}}}
   TriggerRouteTriggerRouteFloat2_1 = "TriggerRoute.Trigger Route (Float, 2):Float"(Value 0=^UE.Attenuation.Distance, Set 0=@n4.Value, Set 1=@n5.Value) {also {defaults {Value 1=0}}}
   n6 = AbsoluteValue.Abs:Float(Input=^Random.Volume.Spread)
   DecibelstoLinearGain_0 = UE.Decibels to Linear Gain:Float(Decibels=UE.RandomFloat(
     Next=^UE.Source.OnPlay
     Max=UE.Add:Float(PrimaryOperand=^Random.Volume.Base, AdditionalOperands=@n6.Value).Out
     Min=UE.Subtract:Float(PrimaryOperand=^Random.Volume.Base, AdditionalOperands=@n6.Value).Out
    ).Value)
   n7 = UE.Reroute(Value=@WavePlayer_0.Out Right)
   n8 = VariableAccessor.Audio(Variable=VariableMutator.Audio(
     Variable=InitVariable.Audio().Variable
     Value=UE.Multiply:Audio by Float(AdditionalOperands=@DecibelstoLinearGain_0.Linear Gain, PrimaryOperand=@n7.Value).Out
    ).Variable)
   n9 = UE.Reroute(Value=@WavePlayer_0.Out Left)
   n10 = VariableAccessor.Audio(Variable=VariableMutator.Audio(
     Variable=InitVariable.Audio().Variable
     Value=UE.Multiply:Audio by Float(AdditionalOperands=@DecibelstoLinearGain_0.Linear Gain, PrimaryOperand=@n9.Value).Out
    ).Variable)
   n11 = /FortMetasound/Attenuation/MSP_Attenuation_Spatialization(
    Azimuth=@TriggerRouteTriggerRouteFloat2_0.Value
    Distance=@TriggerRouteTriggerRouteFloat2_1.Value
    Stereo.InR=@n8.Value
    Stereo.InL=@n10.Value
   ) {also {defaults {StereoSpread=0}}}
   TriggerRouteTriggerRouteBool2_0 = "TriggerRoute.Trigger Route (Bool, 2):Bool"(Set 0=@n4.Value, Set 1=@n5.Value) {also {defaults {Value 0="<MetaSoundBuilderResult.SUCCEEDED: 0>"}}}
   n12 = &MSP_Attenuation_Volume_Curve(Distance=^UE.Attenuation.Distance, bEnable=@TriggerRouteTriggerRouteBool2_0.Value) {also {defaults {AttenuationCurves=/Game/Sounds/Fort_Player/Footsteps/Settings/WTB_FS_3pCurve, FalloffDistance=3584}}}
   n13 = UE.Multiply:Audio by Float(PrimaryOperand=@n11.Right, AdditionalOperands=@n12.VolumeMultiplier)
   TriggerRouteTriggerRouteAudio2_0 = "TriggerRoute.Trigger Route (Audio, 2):Audio"(
    Set 1=UE.Reroute(Value=@n2.False).Value
    Set 0=UE.Reroute(Value=@n2.True).Value
    Value 1=VariableAccessor.Audio(Variable=VariableMutator.Audio(Variable=@n3.Variable, Value=@n13.Out).Variable).Value
    Value 0=VariableAccessor.Audio(Variable=@n8.Variable).Value
   )
   n14 = InitVariable.Audio()
   n15 = UE.Multiply:Audio by Float(PrimaryOperand=@n11.Left, AdditionalOperands=@n12.VolumeMultiplier)
   TriggerRouteTriggerRouteAudio2_1 = "TriggerRoute.Trigger Route (Audio, 2):Audio"(
    Set 1=UE.Reroute(Value=@n2.False).Value
    Set 0=UE.Reroute(Value=@n2.True).Value
    Value 1=VariableAccessor.Audio(Variable=VariableMutator.Audio(Variable=@n14.Variable, Value=@n15.Out).Variable).Value
    Value 0=VariableAccessor.Audio(Variable=@n10.Variable).Value
   )
  ]
  links [
   {dst="@UE.OutputFormat.Stereo.Audio:1", src=TriggerRouteTriggerRouteAudio2_0.Value}
   {dst="@UE.OutputFormat.Stereo.Audio:0", src=TriggerRouteTriggerRouteAudio2_1.Value}
   {dst="@UE.Source.OneShot.OnFinished", src=WavePlayer_0.On Finished}
  ]
 }]
properties {
 assetImportData=&self:AssetImportData
 attenuationSettings=/Game/Sounds/Fort_Player/Footsteps/Settings/Fort_Player_FS_1P_Att
 audioPropertiesSheet=/BRCosmetics/Sounds/Settings/PropertySheets/Kicks/APS_Cosmetic_Kicks_Foley
 bOverrideConcurrency=true
 numChannels=2
 outputFormat=Stereo
 sampleRate=48000
 soundClassObject=&Fort_Player_Foley
 concurrencyOverrides {maxCount=2}
 concurrencySet [null]
}
