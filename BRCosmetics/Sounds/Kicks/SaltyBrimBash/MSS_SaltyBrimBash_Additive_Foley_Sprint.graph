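# Stereo one-shot MetaSound source for the additive sprint foley of the SaltyBrimBash Kicks cosmetic shoes.
# It is a preset of MSS_Kicks_Additive_Foley, feeding a single sprint wave into the high-intensity sound array with randomized pitch.
# All playback logic lives in the parent.
# It uses first-person footstep attenuation, the player foley sound class and a concurrency override capping simultaneous instances.
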MetaSoundSource /BRCosmetics/Sounds/Kicks/SaltyBrimBash/MSS_SaltyBrimBash_Additive_Foley_Sprint
preset_parent=/FortMetasound/Cosmetics/Kicks/MSS_Kicks_Additive_Foley
inputs [
 Character.Affiliation {type=Enum:FortAudioAffiliation}
 Delay Time {type=Time}
 HighIntensity.Sounds {type=WaveAsset:Array}
 Random.Pitch.Base {type=Float}
 Random.Pitch.Spread {type=Float}
 Random.Volume.Base {type=Float}
 Random.Volume.Spread {type=Float}
 UE.Attenuation.Distance {type=Float}
 UE.Source.OnPlay {type=Trigger}
 UE.Spatialization.Azimuth {type=Float}
]
interfaces [UE.OutputFormat.Stereo:1.0, UE.Source.OneShot:1.0, UE.Source:1.1]
outputs [
 "UE.OutputFormat.Stereo.Audio:0" {type=Audio}
 "UE.OutputFormat.Stereo.Audio:1" {type=Audio}
 UE.Source.OneShot.OnFinished {type=Trigger}
]
pages [Default {input_defaults {
   Character.Affiliation=0
   Delay Time=0
   Random.Pitch.Base=1
   Random.Pitch.Spread=0.5
   Random.Volume.Base=0
   Random.Volume.Spread=0
   UE.Attenuation.Distance=0
   UE.Source.OnPlay=null
   UE.Spatialization.Azimuth=0
   HighIntensity.Sounds [&/Shoe_SaltyBrimBash_Sprint_01]
  }}]
properties {
 assetImportData=&self:AssetImportData
 attenuationSettings=/Game/Sounds/Fort_Player/Footsteps/Settings/Fort_Player_FS_1P_Att
 audioPropertiesSheet=/BRCosmetics/Sounds/Settings/PropertySheets/Kicks/APS_Cosmetic_Kicks_Foley
 bOverrideConcurrency=true
 numChannels=2
 outputFormat=Stereo
 sampleRate=48000
 soundClassObject=&Fort_Player_Foley
 volume=0.5
 concurrencyOverrides {maxCount=2}
 concurrencySet [null]
}
